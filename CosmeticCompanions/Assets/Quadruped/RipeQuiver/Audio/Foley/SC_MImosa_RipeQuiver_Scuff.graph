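# Sound cue for scuff foley on the RipeQuiver quadruped cosmetic companion, routed to the creature pets sound class.
# A quality-level switch feeds the first two slots from an equal-weight random pick of ten hard scuff waves, and the last two from a pool of the first four.
# It plays quiet and pitched up at zero priority, using companion foley attenuation and the shared quadruped footstep concurrency.
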SoundCue /CosmeticCompanions/Assets/Quadruped/RipeQuiver/Audio/Foley/SC_MImosa_RipeQuiver_Scuff
root=QualityLevel_0
nodes [
 {id=QualityLevel_0, children [Random_3,Random_3,Random_5,Random_5]}
 {
  id=Random_3
  children [
   WavePlayer_11
   WavePlayer_12
   WavePlayer_13
   WavePlayer_14
   WavePlayer_15
   WavePlayer_16
   WavePlayer_27
   WavePlayer_28
   WavePlayer_29
   WavePlayer_30
  ]
  properties {weights [1,1,1,1,1,1,1,1,1,1]}
 }
 {id=Random_5, children [WavePlayer_31, WavePlayer_32, WavePlayer_33, WavePlayer_34], properties {weights [1,1,1,1]}}
 {
  id=WavePlayer_11
  properties {soundWaveAssetPtr=/CosmeticCompanions/Assets/Quadruped/RipeQuiver/Audio/Foley/Wavs/Scuff/Mimosa_RipeQuiver_Scuff_Hard_01}
 }
 {
  id=WavePlayer_12
  properties {soundWaveAssetPtr=/CosmeticCompanions/Assets/Quadruped/RipeQuiver/Audio/Foley/Wavs/Scuff/Mimosa_RipeQuiver_Scuff_Hard_02}
 }
 {
  id=WavePlayer_13
  properties {soundWaveAssetPtr=/CosmeticCompanions/Assets/Quadruped/RipeQuiver/Audio/Foley/Wavs/Scuff/Mimosa_RipeQuiver_Scuff_Hard_03}
 }
 {
  id=WavePlayer_14
  properties {soundWaveAssetPtr=/CosmeticCompanions/Assets/Quadruped/RipeQuiver/Audio/Foley/Wavs/Scuff/Mimosa_RipeQuiver_Scuff_Hard_04}
 }
 {
  id=WavePlayer_15
  properties {soundWaveAssetPtr=/CosmeticCompanions/Assets/Quadruped/RipeQuiver/Audio/Foley/Wavs/Scuff/Mimosa_RipeQuiver_Scuff_Hard_05}
 }
 {
  id=WavePlayer_16
  properties {soundWaveAssetPtr=/CosmeticCompanions/Assets/Quadruped/RipeQuiver/Audio/Foley/Wavs/Scuff/Mimosa_RipeQuiver_Scuff_Hard_06}
 }
 {
  id=WavePlayer_27
  properties {soundWaveAssetPtr=/CosmeticCompanions/Assets/Quadruped/RipeQuiver/Audio/Foley/Wavs/Scuff/Mimosa_RipeQuiver_Scuff_Hard_07}
 }
 {
  id=WavePlayer_28
  properties {soundWaveAssetPtr=/CosmeticCompanions/Assets/Quadruped/RipeQuiver/Audio/Foley/Wavs/Scuff/Mimosa_RipeQuiver_Scuff_Hard_08}
 }
 {
  id=WavePlayer_29
  properties {soundWaveAssetPtr=/CosmeticCompanions/Assets/Quadruped/RipeQuiver/Audio/Foley/Wavs/Scuff/Mimosa_RipeQuiver_Scuff_Hard_09}
 }
 {
  id=WavePlayer_30
  properties {soundWaveAssetPtr=/CosmeticCompanions/Assets/Quadruped/RipeQuiver/Audio/Foley/Wavs/Scuff/Mimosa_RipeQuiver_Scuff_Hard_10}
 }
 {
  id=WavePlayer_31
  properties {soundWaveAssetPtr=&Mimosa_RipeQuiver_Scuff_Hard_01}
 }
 {
  id=WavePlayer_32
  properties {soundWaveAssetPtr=&Mimosa_RipeQuiver_Scuff_Hard_02}
 }
 {
  id=WavePlayer_33
  properties {soundWaveAssetPtr=&Mimosa_RipeQuiver_Scuff_Hard_03}
 }
 {
  id=WavePlayer_34
  properties {soundWaveAssetPtr=&Mimosa_RipeQuiver_Scuff_Hard_04}
 }
]
properties {
 attenuationSettings=/Game/Sounds/Attenuation/AS_Companion_Foley
 maxDistance=1200
 pitchMultiplier=1.5
 priority=0
 soundClassObject=&Fort_Creatures_Pets
 volumeMultiplier=0.2
 concurrencySet [/CosmeticCompanions/Assets/Audio_Shared/Foley/Quadraped/CONC_Mimosa_Quadraped_Footstep]
}
